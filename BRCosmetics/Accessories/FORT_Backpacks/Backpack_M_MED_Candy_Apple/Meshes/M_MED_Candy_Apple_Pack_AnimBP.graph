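# Animation Blueprint for the Candy Apple backpack cosmetic that positions, tilts and hides the pack on its wearer.
# A cached, offset and scaled ref pose gets extra Root offsets blended in for skydive dive and float states.
# The pack bone's roll follows the owner's spine yaw relative to ref pose, remapped each update.
# Skydive and weapon-change events collapse Root scale outside skydiving when the equipped pickaxe has the HideBackPack tag, and broadcast HideShield.
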AnimBlueprint /BRCosmetics/Accessories/FORT_Backpacks/Backpack_M_MED_Candy_Apple/Meshes/M_MED_Candy_Apple_Pack_AnimBP
target_skeleton=&/M_MED_Candy_Apple_Pack_Skeleton
graphs [
 AnimGraph {nodes [
   n1 = "Transform (Modify) Bone - Bone: Root"(ComponentPose=Local To Component(LocalPose=Use cached pose 'LSPR Pose'()), Alpha=1, Translation="0,-4,0", Rotation="0,0,0", Scale="1,1,1") {
    rotationMode=BMM_Additive
    rotationSpace=BCS_BoneSpace
    scaleMode=BMM_Replace
    scaleSpace=BCS_BoneSpace
    translationMode=BMM_Additive
    translationSpace=BCS_BoneSpace
    alphaBoolBlend {blendOption=Linear}
    boneToModify {boneName=Root}
    translation {y=-4}
   }
   n2 = "Transform (Modify) Bone - Bone: Root"(ComponentPose=Local To Component(LocalPose=Use cached pose 'LSPR Pose'()), Alpha=1, Translation="0,-3,0", Rotation="0,0,0", Scale="1,1,1") {
    rotationMode=BMM_Additive
    rotationSpace=BCS_BoneSpace
    scaleMode=BMM_Replace
    scaleSpace=BCS_BoneSpace
    translationMode=BMM_Additive
    translationSpace=BCS_BoneSpace
    alphaBoolBlend {blendOption=Linear}
    boneToModify {boneName=Root}
    translation {y=-3}
   }
   n3 = Blend Poses by bool(
    BlendPose_0=Component To Local(ComponentPose=@n2)
    BlendPose_1=Blend Poses by bool(
     BlendPose_0=Component To Local(ComponentPose=@n1)
     BlendPose_1=Use cached pose 'LSPR Pose'()
     BlendTime_0=0.15
     BlendTime_1=0.15
     bActiveValue=Get bIsSkydiveFloating()
    ) {blendType=Linear}
    BlendTime_0=0.15
    BlendTime_1=0.15
    bActiveValue=Get bIsSkydiveDiveMode()
   ) {blendType=Linear}
   n4 = Use cached pose 'LSPR Pose'()
   n5 = Get bIsSkyDiving()
   n6 = "Transform (Modify) Bone - Bone: Root"(
    ComponentPose=Local To Component(LocalPose=Blend Poses by bool(BlendPose_0=@n3, BlendPose_1=@n4, BlendTime_0=0.15, BlendTime_1=0.15, bActiveValue=@n5) {blendType=Linear})
    Alpha=Get hiderScale() -> (hiderScale:double)
    Translation="0,0,0"
    Rotation="0,0,0"
    Scale="0,0,0"
   ) {
    rotationMode=BMM_Additive
    scaleMode=BMM_Replace
    scaleSpace=BCS_BoneSpace
    alphaBoolBlend {blendOption=Linear}
    boneToModify {boneName=Root}
    scale=<0 0 0>
   }
   n7 = "Transform (Modify) Bone - Bone: pack"(ComponentPose=@n6, Translation="0,0,0", Rotation=Get backpack_rotate()) {
    rotationMode=BMM_Additive
    rotationSpace=BCS_BoneSpace
    translationMode=BMM_Additive
    alphaBoolBlend {blendOption=Linear}
    boneToModify {boneName=pack}
   }
   Output Pose(Result=Component To Local(ComponentPose=@n7)) {layerGroup=None}
   n8 = "Transform (Modify) Bone - Bone: Root"(ComponentPose=Local To Component(LocalPose=Local Space Ref Pose()), Alpha=1, Translation="0,-3,-8", Rotation="0,0,0", Scale="1.25,1.25,1.25") {
    rotationMode=BMM_Additive
    rotationSpace=BCS_BoneSpace
    scaleMode=BMM_Additive
    scaleSpace=BCS_BoneSpace
    translationMode=BMM_Additive
    translationSpace=BCS_BoneSpace
    alphaBoolBlend {blendOption=Linear}
    boneToModify {boneName=Root}
    scale=<1.25 1.25 1.25>
    translation {y=-3, z=-8}
   }
   Save cached pose 'LSPR Pose'(Pose=Component To Local(ComponentPose=@n8)) {cacheName=LSPR Pose}
  ]}
 EventGraph {nodes [
   Event BlueprintUpdateAnimation()
   n1 = GetOwnerSkeletalMeshComponent()
   n2 = BreakRotator(InRot=BreakTransform(InTransform=GetDeltaTransformFromRefPose(self=@n1, BoneName=spine_05, BaseName=spine_01)).Rotation)
   > Set backpack_rotate(backpack_rotate=MakeRotator(Roll=MapRangeClamped(Value=@n2.Yaw, InRangeA=-14, InRangeB=28, OutRangeA=10, OutRangeB=-5)))
   n3 = Event OnBeginSkydiving()
   n4 = Make Array("[0]"=(TagName="Cosmetics.HideBackPack"))
   n5 = Get bIsSkyDiving()
   n6 = Branch(
    execute=[@n3.then, @n7.then, @n8.then]
    Condition=AND Boolean(A=NOT Boolean(A=@n5), B=PickaxeItemDefinitionHasAnyMetaTag(Tags=@n4, bRequiresEquipped=true))
   )
   > Set hiderScale(hiderScale:double=1) -> (Output_Get:double)
   > Call Hide Shield!(execute:Exec, self:M MED Candy Apple Pack Anim BP Object Reference, isHidden:bool=true) -> (then:Exec) {is=title K2Node_CallDelegate}
   Set hiderScale(execute=@n6.else, hiderScale=0)
   > Call Hide Shield()
   n7 = Event OnWeaponChanged()
   n8 = Event OnEndSkydiving()
   TryGetPawnOwner()
   Get bIsSkyDiving()
  ]}
 HideShield {nodes [HideShield!() -> (then, isHidden:bool=~) {is=title K2Node_FunctionEntry}]}
]
variables [
 backpack_rotate {type=Rotator}
 hiderScale {type=number}
 HideShield {default=()}
]
